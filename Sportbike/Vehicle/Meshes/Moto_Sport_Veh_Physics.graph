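# Physics asset for the sport motorcycle vehicle mesh, giving the bike a simplified collision volume that includes its riders.
# A single Root body holds all shapes, with no other bodies and no constraints.
# Capsules cover the driver, passenger and front shock, while angled boxes cover the rear seat, front fender, right exhaust and rear shock.
# No shapes are defined for the wheels or a left exhaust.
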PhysicsAsset /Sportbike/Vehicle/Meshes/Moto_Sport_Veh_Physics
bodies {Root [
  {shapeType=Capsule, length=46, radius=23.4375, shapeName=DriverCollision, center=<-25.954 0 101.3455>}
  {shapeType=Capsule, length=62.9375, radius=23.4375, shapeName=PassengerCollision, center=<-74.6274 0 112.8158>}
  {
   shapeType=Capsule
   length=33.8125
   radius=14.3438
   shapeName=FrontShock
   center=<60.248 -0.0001 41.7083>
   rotation=<rot 25.3125 -0.0005 -0.0011>
  }
  {
   shapeType=Box
   extentX=60.3125
   extentY=16.625
   extentZ=18.0625
   shapeName=RearSeat
   center=<-81.2167 0 81.3249>
   rotation=<rot -10 0.0001 -0.0006>
  }
  {shapeType=Box, extentX=86, extentY=36.875, extentZ=26.5, shapeName=FrontFender, center=<62.0524 0.0002 69.1558>, rotation=<rot 16.875 0 0>}
  {
   shapeType=Box
   extentX=47.875
   extentY=13.875
   extentZ=18.5
   shapeName=ExhaustRight
   center=<-66.6266 20.5928 34.4466>
   rotation=<rot -7.1875 0.0003 -0.0027>
  }
  {
   shapeType=Box
   extentX=18.75
   extentY=21.125
   extentZ=33.5
   shapeName=RearShock
   center=<-53.5355 0 50.4645>
   rotation=<rot -45 0.0006 -0.0009>
  }
 ]}
constraints []
