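# Physics asset for the sport motorbike vehicle mesh, named as its fluid simulation variant, likely a simplified collision proxy for fluid interaction.
# It approximates the bike entirely with boxes on five bones: two on the body, one on the rear differential, one per tilted wheel, and a tiny one on the frame.
# Even the wheels use boxes, and no constraints link the bodies.
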PhysicsAsset /Sportbike/Vehicle/Meshes/Moto_Sport_Veh_Physics_FluidSim
bodies {
 body [
  {shapeType=Box, extentX=32.822, extentY=54.021, extentZ=179.6944, center=<-13.3941 -0.8519 25.4609>, rotation=<rot -0.0002 -88.7355 -74.5807>}
  {shapeType=Box, extentX=80.5, extentY=25.5, extentZ=35.5, shapeName=body_box, center=<-10.2387 0 -18.4506>, rotation=<rot -20 0.0001 -0.0002>}
 ]
 frame [{shapeType=Box, extentX=1.01, extentY=1.01, extentZ=1.01, rotation=<rot 0 -90 0>}]
 rear_differential [{shapeType=Box, extentX=21.9122, extentY=21.3742, extentZ=84.4019, center=<-19.4292 15.8033 -7.3348>, rotation=<rot 0.001 90 -108.2006>}]
 wheel_tilted_bk [{shapeType=Box, extentX=14.602, extentY=54.8592, extentZ=54.8594, center=<0 1.4646 0>, rotation=<rot 0 -90 0.0001>}]
 wheel_tilted_fr [{shapeType=Box, extentX=10.0067, extentY=52.5333, extentZ=52.5333, center=<-0.0644 0 0.0205>, rotation=<rot 0 -90 -0.0003>}]
}
constraints []
